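# Niagara system for the idle effect of the SaltyBrim Pin pickaxe, a sparse twinkle of short-lived golden sparkle sprites.
# Its single Sparkles emitter spawns randomly sized particles on a rotated, locally offset ring-like shape, nudged by a camera offset.
# Two stacked size curves stretch, swell and shrink each sprite, while a time-of-day scalar adjusts brightness.
# Two sprite renderers draw the particles, with an extra low-alpha orange color attribute from FN_Sprite_Attributes.
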NiagaraSystem /BRCosmetics/Weapons/FORT_Melee/Pickaxe_SaltyBrim_Pin/FX/NS_Pickaxe_SaltyBrim_Pin_Idle
dependencies {
 usedDataInterfaces [
  &NiagaraDataInterfaceCurve
  &NiagaraDataInterfaceVector2DCurve
  &NiagaraDataInterfaceSocketReader
 ]
 usedDynamicInputs []
 usedModules [
  &self:SystemSpawnScript.SystemScriptSource.SystemScriptGraph.NiagaraNodeAssignment_0.SetVariables_#1
  &SystemState
 ]
}
emitters {Sparkles {
  emitterUpdateScript [
   &EmitterState()
   &SpawnRate(SpawnRate=1.5, Use Spawn Probability=false)
  ]
  particleSpawnScript [
   &InitializeParticle(
    Color=<rgba 1 0.5335 0.0003 1>
    Lifetime Max=0.465
    Lifetime Min=0.315
    Lifetime Mode=::NewEnumerator1
    Sprite Size=&Multiply_Vector2D_ByFloat(Float=&RandomRangeFloat(Minimum=9, Maximum=16.5), Vector2D=<1 1>)
    Sprite Size Mode=::NewEnumerator4
    UsePositionOffset=false
    Write Material Random=true
   )
   &CameraOffset(Camera Offset Amount=8)
   &ShapeLocation(
    Handle Radius=2
    Large Radius=10.5
    MaskSpawnGroup=false
    Offset=<-9 -2.4047 0>
    Offset Coordinate Space=::Local
    Offset Mode=::NewEnumerator0
    Override Seed=false
    Rotation Angle Type:ENiagara_AngleInput=::NewEnumerator0
    Rotation Mode=::NewEnumerator2
    Shape Primitive=::NewEnumerator3
    U Distribution=0.2
    UseOwnerScale=true
    Yaw / Pitch / Roll=<15 35 90>
   )
  ]
  particleUpdateScript [
   &ParticleState()
   &ScaleColor(
    Scale RGB=&VectorFromFloat(Value=&FNTimeOfDayScalar(Night Scale=0.65, Morning Scale=0.9, Evening Scale=0.95))
    ScaleA=true
    ScaleRGB=true
   )
   &ScaleSpriteSize(
    Non-Uniform Curve Sprite Scale=DataInterface() {propertyValues {
      XCurve {keys [
        {arriveTangent=-1.8625, interpMode=RCIM_Cubic, leaveTangent=-1.8625, tangentMode=RCTM_User, time=0, value=1}
        {tangentMode=RCTM_User, time=1, value=0.35}
       ]}
      YCurve {keys [
        {arriveTangent=3.1672, interpMode=RCIM_Cubic, leaveTangent=3.1672, tangentMode=RCTM_User, time=0, value=1}
        {arriveTangent=-3.3667, interpMode=RCIM_Cubic, leaveTangent=-3.3667, tangentMode=RCTM_User, time=0.3, value=1.1621}
        {arriveTangent=-0.924, interpMode=RCIM_Cubic, leaveTangent=-0.924, tangentMode=RCTM_User, time=0.5, value=0.7}
        {tangentMode=RCTM_User, time=1, value=0.5}
       ]}
     }}
    Scale Sprite Size Mode=::NewEnumerator3
   )
   &ScaleSpriteSize(Uniform Curve Sprite Scale=DataInterface() {propertyValues {Curve {keys [
        {arriveTangent=1.9457, interpMode=RCIM_Cubic, leaveTangent=1.9457, tangentMode=RCTM_User, time=0, value=0.35}
        {arriveTangent=-3.1633, interpMode=RCIM_Cubic, leaveTangent=-3.1633, tangentMode=RCTM_User, time=0.5, value=1}
        {tangentMode=RCTM_User, time=1, value=0}
       ]}}}) {module=ScaleSpriteSize001}
   &FN_Sprite_Attributes(
    Color=&MakeLinearColorFromVectorAndFloat("Vector (RGB)"=&MakeVectorFromLinearColorRGB(Color=<rgba 1 0.4333 0 1>), "Float (Alpha)"=0.15)
    Create Color Attribute=true
    Module Output:ENiagara_AttributeSamplingApplyOutput=::NewEnumerator1
   )
  ]
  rendererClasses [&NiagaraSpriteRendererProperties]
  renderers [
   {rendererClass=&NiagaraSpriteRendererProperties}
   {rendererClass=&NiagaraSpriteRendererProperties, rendererIndex=1}
  ]
 }}
summary {
 systemName=NS_Pickaxe_SaltyBrim_Pin_Idle
 emitters [{emitterName=Sparkles, rendererClasses [&NiagaraSpriteRendererProperties]}]
 userVariables []
}
