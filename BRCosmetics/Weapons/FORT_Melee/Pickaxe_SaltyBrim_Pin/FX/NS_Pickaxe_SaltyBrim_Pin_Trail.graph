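# Niagara swing trail for the SaltyBrim Pin pickaxe, made of three emitters: Speedlines, Rings and Sparkles.
# Speedlines builds four ribbon anim trails, choosing their socket indices from the swing section index and fading them with a system swing fade.
# Rings spawns short-lived meshes and Sparkles orange sprites, both by distance moved along sampled sockets.
# All three emitters tint or dim by time of day.
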NiagaraSystem /BRCosmetics/Weapons/FORT_Melee/Pickaxe_SaltyBrim_Pin/FX/NS_Pickaxe_SaltyBrim_Pin_Trail
dependencies {
 usedDataInterfaces [
  &NiagaraDataInterfaceColorCurve
  &NiagaraDataInterfaceCurve
  &NiagaraDataInterfaceVector2DCurve
  &NiagaraDataInterfaceSocketReader
 ]
 usedDynamicInputs [&RandomRangeVector]
 usedModules [
  &self:SystemSpawnScript.SystemScriptSource.SystemScriptGraph.NiagaraNodeAssignment_1.SetVariables_#1
  &self:SystemSpawnScript.SystemScriptSource.SystemScriptGraph.NiagaraNodeAssignment_5.SetVariables_#2
  &SystemState
  &self:SystemSpawnScript.SystemScriptSource.SystemScriptGraph.NiagaraNodeAssignment_4.SetVariables_#3
 ]
}
emitters {
 Rings {
  emitterUpdateScript [
   &EmitterState(Life Cycle Mode=::NewEnumerator1, Loop Behavior=::NewEnumerator1, Loop Duration=0.3, UseLoopDelay=false)
   &SpawnPerUnit(
    Module.Use Max Movement Threshold=false
    Module.Use Movement Tolerance=false
    Spawn Probability=1
    Spawn Spacing=40
    Use Spawn Probability=false
   )
  ]
  particleSpawnScript [
   &InitializeParticle(Lifetime=0.14, Mesh Scale Mode=::NewEnumerator3, UsePositionOffset=false, Write Material Random=true)
   &SocketLocation(
    Interpolate Socket Sampling=true
    SamplePosOffset=false
    Socket Reader=$System.Socket Reader
    Socket Sampling Mode=::NewEnumerator3
    VelocitySpeedLimit=false
   )
  ]
  particleUpdateScript [
   &ParticleState()
   &ScaleColor(
    Scale Alpha=&FloatFromCurve(FloatCurve=DataInterface() {propertyValues {Curve {keys [<at 0 1 RCIM_Cubic>,<at 1 0 RCIM_Cubic>]}}})
    Scale RGB=&Multiply_Vector(
     A=&MakeVectorFromLinearColorRGB(Color=&ColorFromCurve(
       ColorCurve=DataInterface() {propertyValues {
         AlphaCurve {keys [<at 0.5 1 RCIM_Cubic>]}
         BlueCurve {keys [
           <at 0.0833 6>
           {arriveTangent=-30.7587, interpMode=RCIM_Cubic, leaveTangent=-30.7587, time=0.125, value=1}
           <at 0.2459 1>
           <at 0.875 1>
           <at 0.9167 6>
          ]}
         GreenCurve {keys [
           <at 0.0833 0.72>
           {arriveTangent=1.7225, interpMode=RCIM_Cubic, leaveTangent=1.7225, time=0.125, value=1}
           <at 0.2459 1>
           <at 0.875 1>
           <at 0.9167 0.72>
          ]}
         RedCurve {keys [
           <at 0.0833 1.359>
           {arriveTangent=-2.2084, interpMode=RCIM_Cubic, leaveTangent=-2.2084, time=0.125, value=1}
           <at 0.2459 1>
           <at 0.875 1>
           <at 0.9167 1.359>
          ]}
        }}
       CurveIndex=Linked() {linkedVariable {name=NPC.FortniteNPC.TimeOfDay_0-1, type=NiagaraFloat}}
      ))
     B=&VectorFromFloat(Value=&FloatFromCurve(FloatCurve=DataInterface() {propertyValues {Curve {keys [<at 0 1 RCIM_Cubic>,<at 1 1.75 RCIM_Cubic>]}}}))
    )
    ScaleA=true
    ScaleRGB=true
   )
   &FN_Mesh_Attributes(
    Color=/Niagara/DynamicInputs/TypeConversions/MakeLinearColorFromFloat(Value:NiagaraFloat=&FloatFromCurve(FloatCurve=DataInterface() {propertyValues {Curve {keys [<at 0 1 RCIM_Cubic>,<at 1 0 RCIM_Cubic>]}}}))
    Create Color Attribute=true
    Create Scale Attribute=true
    Module Output=::NewEnumerator1
    Scale=&Multiply_Vector(A=$Particles.Scale, B=<1.025 1.025 1.025>)
   )
  ]
  rendererClasses [&NiagaraMeshRendererProperties]
  renderers [
   {rendererClass=&NiagaraMeshRendererProperties, rendererIndex=0}
   {rendererClass=&NiagaraMeshRendererProperties, rendererIndex=1}
  ]
 }
 Sparkles {
  emitterUpdateScript [
   &EmitterState(Life Cycle Mode=::NewEnumerator1, Loop Behavior=::NewEnumerator1, Loop Duration=0.3, UseLoopDelay=false)
   &SpawnPerUnit(
    Max Movement Threshold=2500
    Module.Use Max Movement Threshold=false
    Module.Use Movement Tolerance=false
    Spawn Probability=1
    Spawn Spacing=23
    Use Spawn Probability=false
   )
  ]
  particleSpawnScript [
   &InitializeParticle(
    Color=<rgba 1 0.5335 0.0003 1>
    Lifetime Max=0.465
    Lifetime Min=0.315
    Lifetime Mode=::NewEnumerator1
    Sprite Size=&Multiply_Vector2D_ByFloat(Float=&RandomRangeFloat(Minimum=11.2, Maximum=20), Vector2D=<1 1>)
    Sprite Size Mode=::NewEnumerator4
    UsePositionOffset=false
    Write Material Random=true
   )
   &SocketLocation(
    Interpolate Socket Sampling=true
    Orientation Sampling=::NewEnumerator1
    SamplePosOffset=false
    Socket Reader=$System.Socket Reader
    Socket Sampling Mode=::NewEnumerator3
    VelocitySpeedLimit=false
   )
   &ShapeLocation(
    Additional Quaternion Rotation=false
    Handle Radius=2
    Large Radius=10.5
    MaskSpawnGroup=false
    Override Seed=false
    Rotation Mode=::NewEnumerator3
    Rotation Quaternion=$Output.SocketLocation.SampledRotation
    Shape Primitive=::NewEnumerator3
    UseOwnerScale=true
   )
   &CameraOffset(Camera Offset Amount=15)
  ]
  particleUpdateScript [
   &ParticleState()
   &ScaleColor(
    Scale RGB=&VectorFromFloat(Value=&FNTimeOfDayScalar(Night Scale=0.65, Morning Scale=0.9, Evening Scale=0.95))
    ScaleA=true
    ScaleRGB=true
   )
   &ScaleSpriteSize(
    Non-Uniform Curve Sprite Scale=DataInterface() {propertyValues {
      XCurve {keys [
        {arriveTangent=-1.8625, interpMode=RCIM_Cubic, leaveTangent=-1.8625, tangentMode=RCTM_User, time=0, value=1}
        {tangentMode=RCTM_User, time=1, value=0.35}
       ]}
      YCurve {keys [
        {arriveTangent=3.1672, interpMode=RCIM_Cubic, leaveTangent=3.1672, tangentMode=RCTM_User, time=0, value=1}
        {arriveTangent=-3.3667, interpMode=RCIM_Cubic, leaveTangent=-3.3667, tangentMode=RCTM_User, time=0.3, value=1.1621}
        {arriveTangent=-0.924, interpMode=RCIM_Cubic, leaveTangent=-0.924, tangentMode=RCTM_User, time=0.5, value=0.7}
        {tangentMode=RCTM_User, time=1, value=0.5}
       ]}
     }}
    Scale Sprite Size Mode=::NewEnumerator3
   )
   &ScaleSpriteSize(Uniform Curve Sprite Scale=DataInterface() {propertyValues {Curve {keys [
        {arriveTangent=1.9457, interpMode=RCIM_Cubic, leaveTangent=1.9457, tangentMode=RCTM_User, time=0, value=0.35}
        {arriveTangent=-3.1633, interpMode=RCIM_Cubic, leaveTangent=-3.1633, tangentMode=RCTM_User, time=0.5, value=1}
        {tangentMode=RCTM_User, time=1, value=0}
       ]}}}) {module=ScaleSpriteSize001}
   &FN_Sprite_Attributes(
    Color=&MakeLinearColorFromVectorAndFloat("Vector (RGB)"=&MakeVectorFromLinearColorRGB(Color=<rgba 1 0.4333 0 1>), "Float (Alpha)"=0.07)
    Create Color Attribute=true
    Module Output:ENiagara_AttributeSamplingApplyOutput=::NewEnumerator1
   )
  ]
  rendererClasses [&NiagaraSpriteRendererProperties]
  renderers [
   {rendererClass=&NiagaraSpriteRendererProperties}
   {rendererClass=&NiagaraSpriteRendererProperties, rendererIndex=1}
  ]
 }
 Speedlines {
  emitterSpawnScript [&FN_RibbonIDOffsetUtility()]
  emitterUpdateScript [
   &EmitterState()
   &FN_SpawnAnimTrail(Number of Anim Trails=4)
  ]
  particleSpawnScript [
   &FN_InitializeAnimTrailBoneIndecies(First Bone or Socket Index=1)
   &FN_InitializeAnimTrailBoneIndecies(
    AnimTrail Selection=1
    First Bone or Socket Index:NiagaraInt32=/Niagara/DynamicInputs/Helpers/MaskInt(
     Mask Source:NiagaraInt32=1
     Mask Parameter:NiagaraInt32=$User.SwingSectionIndex
     Mask:NiagaraInt32=1
     Masked Int:NiagaraInt32=7
     Passthrough Int:NiagaraInt32=2
    )
    Second Bone or Socket Index:NiagaraInt32=&MaskInt(
     Mask Source:NiagaraInt32=1
     Mask Parameter:NiagaraInt32=$User.SwingSectionIndex
     Mask:NiagaraInt32=1
     Masked Int:NiagaraInt32=7
     Passthrough Int:NiagaraInt32=2
    )
   ) {module=FN_InitializeAnimTrailBoneIndecies001}
   &FN_InitializeAnimTrailBoneIndecies(
    AnimTrail Selection=2
    First Bone or Socket Index:NiagaraInt32=&MaskInt(
     Mask Source:NiagaraInt32=1
     Mask Parameter:NiagaraInt32=$User.SwingSectionIndex
     Mask:NiagaraInt32=2
     Masked Int:NiagaraInt32=5
     Passthrough Int:NiagaraInt32=3
    )
    Second Bone or Socket Index:NiagaraInt32=&MaskInt(
     Mask Source:NiagaraInt32=1
     Mask Parameter:NiagaraInt32=$User.SwingSectionIndex
     Mask:NiagaraInt32=2
     Masked Int:NiagaraInt32=5
     Passthrough Int:NiagaraInt32=3
    )
   ) {module=FN_InitializeAnimTrailBoneIndecies002}
   &FN_InitializeAnimTrailBoneIndecies(
    AnimTrail Selection=3
    First Bone or Socket Index:NiagaraInt32=&MaskInt(
     Mask Source:NiagaraInt32=1
     Mask Parameter:NiagaraInt32=$User.SwingSectionIndex
     Mask:NiagaraInt32=2
     Masked Int:NiagaraInt32=6
     Passthrough Int:NiagaraInt32=4
    )
    Second Bone or Socket Index:NiagaraInt32=&MaskInt(
     Mask Source:NiagaraInt32=1
     Mask Parameter:NiagaraInt32=$User.SwingSectionIndex
     Mask:NiagaraInt32=2
     Masked Int:NiagaraInt32=6
     Passthrough Int:NiagaraInt32=4
    )
   ) {module=FN_InitializeAnimTrailBoneIndecies003}
   &FN_InitializeAnimTrail(
    AnimTrailWidth=&MergeHelper_FloatSwitch(0=0.8, 1=1, 2=1.15, 3=0.875, Index=&SpawnGroup(), Num=::NewEnumerator2)
    Color=<rgba 1 0.3 0 1>
    Lifetime=0.22
   )
  ]
  particleUpdateScript [
   &FN_UpdateAnimTrail() {module=FN_UpdateAnimTrail001}
   &ScaleColor(Scale Alpha=$System.SwingGeneralFade, ScaleA=true, ScaleRGB=true) {module=ScaleColor001}
   &DynamicMaterialParameters(
    Index 0 Param 1=1
    Index 0 Param 3=1
    Index 2 Param 1=100
    Index 2 Param 2=&MergeHelper_FloatSwitch(0=4, 1=8, 2=14, 3=11, Index=&SpawnGroup(), Num=::NewEnumerator2)
    Index 2 Param 4=0.6
    Module.Index0Float0Write=true
    Module.Index0Float1Write=true
    Module.Index0Float2Write=true
    Module.Index0Float3Write=true
    Module.Index1Float0Write=true
    Module.Index1Float1Write=true
    Module.Index1Float2Write=true
    Module.Index1Float3Write=true
    Module.Index2Float0Write=true
    Module.Index2Float1Write=true
    Module.Index2Float2Write=false
    Module.Index2Float3Write=false
    Write Parameter Index 1=::NewEnumerator1
    Write Parameter Index 2=::NewEnumerator1
   )
   &ScaleColor(
    Scale RGB=&MakeVectorFromLinearColorRGB(Color=&ColorFromCurve(
      ColorCurve=DataInterface() {propertyValues {
        AlphaCurve {keys [<at 0 1.3>,<at 0.125 1.2>,<at 0.25 1.1>,<at 0.5 1>,<at 0.75 1.1>,<at 0.875 1.2>,<at 1 1.3>]}
        BlueCurve {keys [<at 0.0417 1>,<at 0.125 0.8304>,<at 0.25 0.7>,<at 0.375 0.935>,<at 0.5 0.895>,<at 0.625 0.935>,<at 0.75 0.7>,<at 0.875 0.675>,<at 0.9167 1>,<at 1 1>]}
        GreenCurve {keys [<at 0.0417 0.5159>,<at 0.125 0.6609>,<at 0.25 0.731>,<at 0.375 0.935>,<at 0.5 0.895>,<at 0.625 0.935>,<at 0.75 0.7>,<at 0.875 0.7467>,<at 0.9167 0.6376>,<at 1 0.5159>]}
        RedCurve {keys [<at 0.0417 0.402>,<at 0.125 1>,<at 0.25 1>,<at 0.375 0.935>,<at 0.5 0.895>,<at 0.625 0.935>,<at 0.75 1>,<at 0.875 1>,<at 0.9167 0.6>,<at 1 0.402>]}
       }}
      CurveIndex=Linked() {linkedVariable {name=NPC.FortniteNPC.TimeOfDay_0-1, type=NiagaraFloat}}
     ))
    ScaleA=true
    ScaleRGB=true
   )
  ]
  rendererClasses [&NiagaraRibbonRendererProperties]
  renderers [{rendererClass=&NiagaraRibbonRendererProperties}]
 }
}
summary {
 systemName=NS_Pickaxe_SaltyBrim_Pin_Trail
 emitters [
  {emitterName=Speedlines, rendererClasses [&NiagaraRibbonRendererProperties]}
  {emitterName=Rings, rendererClasses [&NiagaraMeshRendererProperties]}
  {emitterName=Sparkles, rendererClasses [&NiagaraSpriteRendererProperties]}
 ]
 userVariables [
  User.bIsFrontEnd {type=NiagaraBool}
  User.bIsFrontEndPreview {type=NiagaraBool}
  User.RibbonIDOffset {type=NiagaraInt32}
  User.SwingSectionIndex {type=NiagaraInt32}
 ]
}
